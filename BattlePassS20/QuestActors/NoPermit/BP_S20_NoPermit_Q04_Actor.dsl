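; Season 20 Battle Pass "NoPermit" Q04 quest prop that shows a radar tower mesh cloaked, decloaking it when an eligible player comes near.
; BeginPlay applies cloaked materials; client-only cosmetic enter and exit events swap materials and trigger quest VFX and audio.
; Cloak events write game time and a state flag into custom primitive data for the materials.
; ShouldToggleCosmetics gates toggling on quest state or test flag, interactability, and discover-collision unhide being off.
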(event EventBeginPlay
  (bind _sm_quest_preinteraction (Variables|Components|GetSM_Quest_PreInteraction))
  (|CallParentFunction)
  (Rendering|Material|SetMaterial _sm_quest_preinteraction 0 "/Game/Missions/Secondary/BuildRadarTower/Materials/M_BuildRadarTower_Bottom_IO_Cloaked.M_BuildRadarTower_Bottom_IO_Cloaked")
  (Rendering|Material|SetMaterial _sm_quest_preinteraction 1 "/Game/Missions/Secondary/BuildRadarTower/Materials/M_BuildRadarTower_Top_IO_Cloaked.M_BuildRadarTower_Top_IO_Cloaked"))

(event Custom|Cloaking%20End
  (bind _sm_quest_preinteraction (Variables|Components|GetSM_Quest_PreInteraction))
  (bind _sm_quest_preinteraction_2 (Variables|Components|GetSM_Quest_PreInteraction))
  ;[1 1 1] Cloaking Logic
  (Rendering|Material|SetMaterial _sm_quest_preinteraction_2 0 "/Game/Missions/Secondary/BuildRadarTower/Materials/M_BuildRadarTower_Bottom_IO_Cloaked.M_BuildRadarTower_Bottom_IO_Cloaked")
  (Rendering|Material|SetMaterial _sm_quest_preinteraction_2 1 "/Game/Missions/Secondary/BuildRadarTower/Materials/M_BuildRadarTower_Top_IO_Cloaked.M_BuildRadarTower_Top_IO_Cloaked")
  (Rendering|Material|SetCustomPrimitiveDataFloat _sm_quest_preinteraction 0 (Utilities|Time|GetGameTimeinSeconds))
  (Rendering|Material|SetCustomPrimitiveDataFloat _sm_quest_preinteraction 1 0.0))

(event Custom|Cloaking%20State%20Change
  (bind _sm_quest_preinteraction (Variables|Components|GetSM_Quest_PreInteraction))
  ;[1 1 1] Cloaking Logic
  (Rendering|Material|SetCustomPrimitiveDataFloat _sm_quest_preinteraction 0 (Utilities|Time|GetGameTimeinSeconds))
  (Rendering|Material|SetCustomPrimitiveDataFloat _sm_quest_preinteraction 1 1.0))

(event Custom|Decloaking%20End
  (bind _sm_quest_preinteraction (Variables|Components|GetSM_Quest_PreInteraction))
  (bind _sm_quest_preinteraction_2 (Variables|Components|GetSM_Quest_PreInteraction))
  ;[1 1 1] Cloaking Logic
  (Rendering|Material|SetMaterial _sm_quest_preinteraction 0 "/Game/Missions/Secondary/BuildRadarTower/Materials/M_BuildRadarTower_Bottom_IO.M_BuildRadarTower_Bottom_IO")
  (Rendering|Material|SetMaterial _sm_quest_preinteraction 1 "/Game/Missions/Secondary/BuildRadarTower/Materials/M_BuildRadarTower_Top_IO.M_BuildRadarTower_Top_IO")
  (Rendering|Material|SetCustomPrimitiveDataFloat _sm_quest_preinteraction_2 0 (Utilities|Time|GetGameTimeinSeconds))
  (Rendering|Material|SetCustomPrimitiveDataFloat _sm_quest_preinteraction_2 1 0.0))

(event EventOnEnterCosmetic
  (params (PlayerController :type Fort%20Player%20Controller%20Object%20Reference) (PlayerPawn :type Fort%20Player%20Pawn%20Athena%20Object%20Reference))
  (Networking|SwitchHasAuthority
    (:Remote
      (bind _returnvalue (CallFunction|ShouldToggleCosmetics PlayerController PlayerPawn))
      (if _returnvalue
        (VisualState|VFX&Audio|PreInteractionVFX&Audio PlayerController)
        (VisualState|VFX&Audio|UnhideVFX&Audio PlayerController)
        (CallFunction|CloakingStateChange)
        (CallFunction|DecloakingEnd)))))

(event EventOnExitCosmetic
  (params (PlayerController :type Fort%20Player%20Controller%20Object%20Reference) (PlayerPawn :type Fort%20Player%20Pawn%20Athena%20Object%20Reference))
  (|CallParentFunction PlayerController PlayerPawn)
  (Networking|SwitchHasAuthority
    (:Remote
      (bind _returnvalue (CallFunction|ShouldToggleCosmetics PlayerController PlayerPawn))
      (if _returnvalue
        (VisualState|VFX&Audio|UnhideVFX&Audio PlayerController)
        (CallFunction|CloakingStateChange)
        (CallFunction|CloakingEnd)))))

(fn ShouldToggleCosmetics
  (params (PlayerController :type Fort%20Player%20Controller%20Object%20Reference) (InteractingPawn :type Fort%20Pawn%20Object%20Reference))
  (bind _asfort_quest_item_definition_athena (Utilities|Casting|CastToFortQuestItemDefinition_Athena (Class|ConsolidatedQuestComponent|GetQuestItemDefinition (Variables|BuildingGameplayActorPropQuest|GetConsolidatedQuestComponent))))
  (bind (_playercontrollerout _completedquest) (Quest|HasPlayerCompletedQuest PlayerController _asfort_quest_item_definition_athena))
  (return (Math|Boolean|ANDBoolean (not (Variables|Design|GetbUseDiscoverCollisionToUnhide)) (or (and (not _completedquest) (ObjectiveState|IsObjectiveActive PlayerController)) (and (Variables|Design|QuestSetup|GetbTestVisualsNoQuestRequired) (not (Variables|Advanced|Debug|GetbOnVisualTestingCooldown)))) (Interaction|BlueprintCanInteract InteractingPawn "AllInteraction"))))

(fn ConstructionScript
  (|CallParentFunction))
