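# Sound cue for the cloak sound of the dish actor in the Battle Pass S20 NoPermit quest.
# A random node picks one of two wave variations with equal weight.
# The cue routes through the Fort_Quests sound class and overrides attenuation with its own falloff, stereo spread and shape extent, limiting how far the sound carries.
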SoundCue /BattlePassS20/QuestActors/NoPermit/Sounds/Quest_NoPermit_Dish_Cloak_Cue
root=Random_1
nodes [
 {id=Random_1, children [WavePlayer_8,WavePlayer_9], properties {weights [1,1]}}
 {id=WavePlayer_8, properties {soundWaveAssetPtr=&/Quest_NoPermit_Dish_Cloak_01}}
 {id=WavePlayer_9, properties {soundWaveAssetPtr=&/Quest_NoPermit_Dish_Cloak_02}}
]
properties {
 bOverrideAttenuation=true
 maxDistance=2200
 soundClassObject=&Fort_Quests
 volumeMultiplier=1
 attenuationOverrides {falloffDistance=2000, stereoSpread=500, attenuationShapeExtents {x=200}}
}
