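# Sound cue for the destruction of a dish in the NoPermit quest, on the impacts sound class.
# It mixes the dish-destroyed wave, randomly faded by one of two volume envelopes and randomly pitched up, with a briefly delayed random workbench junk hit.
# A root modulator wraps the mix.
# The cue overrides attenuation and uses prevent-new concurrency.
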SoundCue /BattlePassS20/QuestActors/NoPermit/Sounds/Quest_NoPermit_Dish_Destroyed_Cue
root=Modulator_1
nodes [
 {id=Delay_0, children [Random_3], properties {delayMax=0.2, delayMin=0.15}}
 {
  id=Enveloper_0
  children [WavePlayer_8]
  properties {pitchCurve {editorCurveData {keys [{time=0}]}}, volumeCurve {editorCurveData {keys [<at 0 1>,<at 0.436 0.993>,<at 1.1825 -0.0124>]}}}
 }
 {
  id=Enveloper_3
  children [WavePlayer_8]
  properties {pitchCurve {editorCurveData {keys [{time=0}]}}, volumeCurve {editorCurveData {keys [<at 0 1>,<at 0.1791 0.9929>,<at 0.8365 -0.0095>]}}}
 }
 {id=Mixer_0, children [Modulator_0,Delay_0], properties {inputVolume [1,1]}}
 {id=Modulator_0, children [Random_2], properties {pitchMax=1.2, pitchMin=1, volumeMax=1, volumeMin=1}}
 {id=Modulator_1, children [Mixer_0]}
 {id=Random_2, children [Enveloper_0,Enveloper_3], properties {weights [1,1]}}
 {id=Random_3, children [WavePlayer_2, WavePlayer_9, WavePlayer_10], properties {weights [1,1,1]}}
 {id=WavePlayer_10, properties {soundWaveAssetPtr=/Game/Sounds/Fort_Impact_Sounds/Demolition/Workbench/Workbench_Junk_Hit2}}
 {id=WavePlayer_2, properties {soundWaveAssetPtr=/Game/Sounds/Fort_Impact_Sounds/Demolition/Workbench/Workbench_Junk_Hit3}}
 {id=WavePlayer_8, properties {soundWaveAssetPtr=&/Quest_NoPermit_Dish_Destroyed_01}}
 {id=WavePlayer_9, properties {soundWaveAssetPtr=/Game/Sounds/Fort_Impact_Sounds/Demolition/Workbench/Workbench_Junk_Hit1}}
]
properties {
 bOverrideAttenuation=true
 bOverrideConcurrency=true
 maxDistance=3000
 soundClassObject=&Fort_Impacts
 volumeMultiplier=1.25
 attenuationOverrides {falloffDistance=2400, stereoSpread=1000, attenuationShapeExtents {x=600}}
 concurrencyOverrides {resolutionRule=PreventNew}
}
