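# Sound cue for completion of the fourth NoPermit quest in the BattlePassS20 content, going by its name.
# A single Wave Player at the root plays an Origin VO stinger wave, with no randomization, mixing or other nodes.
# It is routed to the quest music sound class with a slight volume boost and a very large max distance.
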SoundCue /BattlePassS20/QuestActors/NoPermit/Sounds/Quest_NoPermit_Q4_Complete_Cue
root=WavePlayer_0
nodes [{id=WavePlayer_0, properties {soundWaveAssetPtr=&/FNS20_Origin_VO_Stinger_02_and_03_110_BPM}}]
properties {maxDistance=2097152, soundClassObject=/Game/Sounds/FortSoundClassesAndModes/Fort_Music_Quest_PSM, volumeMultiplier=1.2}
